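# Translucent, unlit, two-sided trail material for Niagara mesh particles on the ChimeCurlCorn backpack.
# Panning noise, scaled by dynamic parameters, distorts the trail texture UVs, with scalar lerps selecting between two UV setups.
# A trail channel, particle alpha, optional panning mask and depth fade form opacity; particle color, a scrolling color texture and optional noise-driven tint form emissive.
# Two opposing sine waves along U build a Z offset scaled by WPO_Int.
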Material /BRCosmetics/Accessories/FORT_Backpacks/Backpack_F_MED_ChimeCurlCorn/FX/M_Backpack_ChimeCurlCorn_Trail
graph {
 properties {
  bAutomaticallySetUsageInEditor=false
  bUsedWithNiagaraMeshParticles=true
  blendMode=BLEND_Translucent
  shadingModel=MSM_Unlit
  twoSided=true
  usedShadingModels=Unlit
 }
 nodes [
  n1 = ParticleColor()
  n2 = TextureCoordinate()
  n3 = AppendVector(
   A=ScalarParameter() {defaultValue=3, parameterName=Trail Tex Tiling}
   B=ScalarParameter() {defaultValue=0.8, parameterName=Trail Tex TilingV}
  )
  n4 = Constant()
  n5 = ScalarParameter() {defaultValue=-0.1, parameterName=TrailTex_OffsetV}
  n6 = TextureCoordinate()
  n7 = AppendVector(A=ScalarParameter() {parameterName=Noise A offsetU}, B=Constant())
  n8 = ScalarParameter() {defaultValue=3, parameterName=Noise A Tiling}
  n9 = Constant()
  n10 = MaterialFunctionCall(XY=Multiply(A=Add(A=@n6, B=@n7), B=AppendVector(A=@n8, B=@n9))) {materialFunction=&Swizzle}
  n11 = AppendVector(A=Constant(), B=ScalarParameter() {defaultValue=0.3, parameterName=Noise A Speed})
  n12 = TextureCoordinate()
  n13 = AppendVector(A=ScalarParameter() {parameterName=Noise B offsetU}, B=Constant())
  n14 = ScalarParameter() {defaultValue=8, parameterName=Noise B Tiling}
  n15 = Constant()
  n16 = MaterialFunctionCall(XY=Multiply(A=Add(A=@n12, B=@n13), B=AppendVector(A=@n14, B=@n15))) {materialFunction=&Swizzle}
  n17 = AppendVector(A=Constant(), B=ScalarParameter() {defaultValue=-0.2, parameterName=Noise B Speed})
  n18 = Add(
   A=TextureSampleParameter2D(UVs=Panner(Coordinate=@n10.YX, Speed=@n11) {speedY=4}) {parameterName=TrailNoise_TexA, texture=/BRCosmetics/Weapons/FORT_Melee/Pickaxe_SoapPocket/FX/T_Pickaxe_SoapPocket_Noise}.R
   B=TextureSampleParameter2D(UVs=Panner(Coordinate=@n16.YX, Speed=@n17) {speedY=4}) {parameterName=TrailNoise_TexB, texture=/Game/Global/Textures/Default/Masks_and_Gradients/gradient_stasis}.R
  )
  n19 = Time()
  n20 = Time()
  n21 = LinearInterpolate(Alpha=Multiply(A=Sine(None=@n19) {period=0.4}, B=Sine(None=@n20) {period=0.7})) {constA=0.8, constB=1.2}
  n22 = Multiply(A=@n18, B=@n21)
  n23 = ScalarParameter() {parameterName=Use_NoiseSine}
  n24 = Add(
   A=Add(A=Multiply(A=@n2, B=@n3), B=AppendVector(A=@n4, B=@n5))
   B=Multiply(
    A=LinearInterpolate(A=@n22, B=@n18, Alpha=@n23)
    B=DynamicParameter() {defaultValue=<rgba 0.2 1 1 1>, paramNames [Noise_Int,WPO_Int,Noise02,Param4]}.Noise_Int
   ) {constB=0.2}
  )
  n25 = Add(A=TextureCoordinate() {vTiling=0.5}, B=Constant2Vector() {g=0.1})
  n26 = Panner(Coordinate=MaterialFunctionCall(XY=TextureCoordinate() {uTiling=3, vTiling=0}) {materialFunction=&Swizzle}.YX) {speedY=0.2}
  n27 = Constant() {r=0.6}
  n28 = DynamicParameter() {defaultValue=<rgba 0.2 1 1 1>, paramNames [Noise_Int,WPO_Int,Noise02,Param4]}
  n29 = TextureCoordinate() {uTiling=3, vTiling=0}
  n30 = Constant2Vector() {r=0.05}
  n31 = Panner(Coordinate=MaterialFunctionCall(XY=Add(A=@n29, B=@n30)) {materialFunction=&Swizzle}.YX) {speedY=0.2}
  n32 = Constant() {r=-0.3}
  n33 = Add(
   A=Multiply(
    A=TextureSample(UVs=@n26) {texture=/BRCosmetics/Characters/Player/Female/Medium/Bodies/F_MED_LocalTalent_Flare/FX/T_Character_LocalTalent_Flare_Noise}.R
    B=Multiply(A=@n27, B=@n28.Noise02)
   ) {constB=0.5}
   B=Multiply(
    A=TextureSample(UVs=@n31) {texture=&T_Character_LocalTalent_Flare_Noise}.R
    B=Multiply(A=@n32, B=@n28.Noise02)
   ) {constB=-0.15}
  )
  n34 = TextureSample(UVs=Panner(Coordinate=TextureCoordinate() {uTiling=2, vTiling=0}) {speedY=-0.4}) {texture=&T_Pickaxe_SoapPocket_Noise}
  n35 = LinearInterpolate(A=@n33, B=Multiply(A=@n21, B=Multiply(A=@n33, B=@n34.R)), Alpha=@n23)
  n36 = TextureSampleParameter2D!(
   UVs=LinearInterpolate(A=Panner(Coordinate=@n24), B=Add(A=@n25, B=@n35), Alpha=ScalarParameter() {parameterName=Use_SecondUVcontrol})
   Level
   Apply View MipBias
  ) {
   group=Trail
   mipValueMode=TMVM_MipLevel
   parameterName=Trail Tex
   sortPriority=0
   texture=&/T_Backpack_ChimeCurlCorn_Trail
  }
  n37 = Power(Base=@n36.RGBA, Exp=ScalarParameter() {defaultValue=1, parameterName=TrailTex_Power})
  n38 = Multiply(
   A=@n1.A
   B=ChannelMaskParameter(None=@n37) {group=Trail, maskChannel=Green, parameterName=Trail Alpha Channel, defaultValue=<rgba 0 1 0 0>}.R
  )
  n39 = ScalarParameter() {parameterName=TexMask_OffsetU}
  n40 = Constant()
  n41 = VectorParameter() {parameterName=TexMask, defaultValue=<rgba 3 1 0 0>}
  n42 = Multiply(A=Add(A=TextureCoordinate(), B=AppendVector(A=@n39, B=@n40)), B=AppendVector(A=@n41.R, B=@n41.R))
  n43 = TextureSampleParameter2D!(
   UVs=Panner(
    Coordinate=MaterialFunctionCall(XY=@n42) {materialFunction=&Swizzle}.YX
    Speed=AppendVector(A=@n41.B, B=@n41.B)
   )
   Apply View MipBias
  ) {parameterName=TexMask, texture=/Game/Global/Textures/Default/Masks_and_Gradients/T_Water_Top_Ramp}
  n44 = Saturate(None=Power(Base=Multiply(A=@n43.R) {constB=3}))
  DepthFade(
   Opacity=LinearInterpolate(A=@n38, B=Multiply(A=@n38, B=@n44), Alpha=ScalarParameter() {parameterName=Use_TexMask})
   FadeDistance=ScalarParameter() {defaultValue=5, parameterName=DepthFade}
  )
  n45 = TextureCoordinate() {uTiling=3, vTiling=0.5}
  n46 = Multiply(
   A=VectorParameter() {parameterName=ColorTex_Color, defaultValue=<rgba 1 0.05 0.05 1>}.RGB
   B=TextureSample(UVs=Panner(Coordinate=@n45) {speedY=1}) {texture=&T_Tile_RainbowColor}.B
  )
  n47 = ScalarParameter() {defaultValue=1, parameterName=Desaturation}
  n48 = MaterialFunctionCall(
   In=ChannelMaskParameter(None=@n37) {group=Trail, maskChannel=Green, parameterName=Trail Alpha Channel, defaultValue=<rgba 0 1 0 0>}.R
   Contrast=ScalarParameter() {group=Emissive, parameterName=Emissive Contrast}
  ) {materialFunction=&CheapContrast}
  n49 = Add(
   A=Multiply(A=Desaturation(None=@n46, Fraction=@n47), B=ScalarParameter() {defaultValue=1.5, parameterName=ColorTex_Mul}) {constB=1.5}
   B=LinearInterpolate(A=Multiply(A=@n1.RGB, B=@n48.Result), B=@n1.RGB, Alpha=ScalarParameter() {parameterName=Use_NoAlphaTex})
  )
  n50 = TextureSampleParameter2D(UVs=Panner(Coordinate=TextureCoordinate() {uTiling=2, vTiling=0}) {speedY=-0.4}) {
   parameterName=ColorLerp_Texture
   samplerType=SAMPLERTYPE_Grayscale
   texture=/BRCosmetics/Weapons/FORT_Melee/Pickaxe_Cube_Ninja_Male/FX/T_CellNoise
  }
  n51 = LinearInterpolate(
   A=VectorParameter() {parameterName=ColorLerpA, defaultValue=<rgba 0 0.4236 0.5 1>}.RGB
   B=VectorParameter() {parameterName=ColorLerpB, defaultValue=<rgba 0.2031 1.8059 3 1>}.RGB
   Alpha=@n50.R
  )
  MaterialFunctionCall(Emissive=LinearInterpolate(A=@n49, B=Multiply(A=@n51, B=@n49), Alpha=ScalarParameter() {parameterName=Use_ColorCurve})) {materialFunction=&MF_FX_EmissiveAdjustments}
  n52 = CurveAtlasRowParameter(CurveTime=ComponentMask(None=TextureCoordinate()) {r=true}) {
   atlas=/Game/Accessories/FORT_Backpacks/Backpack_F_MED_WaywardRebel/FX/CAtlas_WaywardRebel_ColorCurve
   curve=/Game/Characters/Player/Female/Medium/Bodies/F_MED_WaywardRebel/FX/CV_WaywardRebel_Skin_ColorCurve01
   defaultValue=1
   parameterName=Color001
  }
  Multiply(A=@n52)
  n53 = Constant()
  n54 = ComponentMask(None=TextureCoordinate()) {r=true}
  n55 = ScalarParameter() {defaultValue=12, group=WPO, parameterName=WPO A Tiling}
  n56 = Time()
  n57 = ScalarParameter() {defaultValue=4, group=WPO, parameterName=WPO A Speed}
  n58 = Sine(None=Add(A=Multiply(A=@n54, B=@n55) {constB=0.5}, B=Multiply(A=@n56, B=@n57) {constB=3}))
  n59 = Multiply(
   A=MaterialFunctionCall(X=@n53, Y=@n53, Z=LinearInterpolate(Alpha=@n58) {constA=-1}) {materialFunction=&MakeFloat3}.Result
   B=ScalarParameter() {defaultValue=1, group=WPO, parameterName=Wave A Intensity}
  )
  n60 = ScalarParameter() {defaultValue=8, group=WPO, parameterName=WPO B Tiling}
  n61 = ScalarParameter() {defaultValue=-2, group=WPO, parameterName=WPO B Speed}
  n62 = Sine(None=Add(A=Multiply(A=@n54, B=@n60) {constB=0.5}, B=Multiply(A=@n56, B=@n61) {constB=5}))
  n63 = Multiply(
   A=MaterialFunctionCall(X=@n53, Y=@n53, Z=LinearInterpolate(Alpha=@n62) {constA=-1}) {materialFunction=&MakeFloat3}.Result
   B=ScalarParameter() {defaultValue=-1, group=WPO, parameterName=Wave B Intensity}
  )
  Reroute(None=Multiply(A=DynamicParameter() {defaultValue=<rgba 0.2 1 1 1>, paramNames [Noise_Int,WPO_Int,Noise02,Param4]}.WPO_Int, B=Add(A=@n59, B=@n63)) {constA=1})
 ]
}
