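# Idle Niagara effect for the ChimeCurlCorn back bling, an audio-reactive, mostly red glow with music lines and sparkles.
# Persistent Glow (with light), MusicLine and MusicLineGlow layers fade and scale with an interpolated moving average.
# BoxGlow, Glints, Star and spikes gate spawn probability on System.Audio Reactive.
# Event audio amplitude, often raised to a high power, drives peak flares; front-end preview keeps some layers faintly lit.
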NiagaraSystem /BRCosmetics/Accessories/FORT_Backpacks/Backpack_F_MED_ChimeCurlCorn/FX/NS_Backpack_ChimeCurlCorn_Idle
dependencies {
 usedDataInterfaces [
  &NiagaraDataInterfaceArrayFloat3
  &NiagaraDataInterfaceArrayColor
  &NiagaraDataInterfaceCurve
  &NiagaraDataInterfaceArrayFloat
  &NiagaraDataInterfaceColorCurve
  &NiagaraDataInterfaceVector2DCurve
  &NiagaraDataInterfaceSkeletalMesh
 ]
 usedDynamicInputs []
 usedModules [
  &self:SystemSpawnScript.SystemScriptSource.SystemScriptGraph.NiagaraNodeAssignment_2.SetVariables_#1
  &SystemState
  &InterpolateOverTime
  &self:SystemSpawnScript.SystemScriptSource.SystemScriptGraph.NiagaraNodeAssignment_1.SetVariables_#2
  &self:SystemSpawnScript.SystemScriptSource.SystemScriptGraph.NiagaraNodeAssignment_0.SetVariables_#3
 ]
}
emitters {
 BoxGlow {
  emitterUpdateScript [
   &EmitterState(Inactive Response=::NewEnumerator1, Life Cycle Mode=::NewEnumerator1, Loop Duration=1, UseLoopDelay=true)
   &SpawnRate(Spawn Probability=&Lerp_Float(A=0, B=1, Alpha=$System.Audio Reactive), SpawnRate=5, Use Spawn Probability=true)
  ]
  particleSpawnScript [
   &InitializeParticle(
    Color=<rgba 1.6 1.6 1.6 1>
    Mesh Scale Max=<1 1 2.5>
    Mesh Scale Min=<1 1 1.5>
    Mesh Scale Mode=::NewEnumerator2
    Position Offset=<0 -12 -5>
    Sprite Rotation Mode=::NewEnumerator2
    Sprite Size Mode=::NewEnumerator1
    Uniform Sprite Size Max=15
    Uniform Sprite Size Min=10
    UsePositionOffset=true
    Write Material Random=true
   )
   &ShapeLocation(
    Cylinder Height=1
    Cylinder Radius=5
    MaskSpawnGroup=true
    Offset Mode=::NewEnumerator0
    Override Seed=false
    Shape Primitive=::NewEnumerator1
    Spawn Group:NiagaraInt32=$Particles.SpawnGroup
    Surface Only=false
    UseOwnerScale=true
   )
   &InitialMeshOrientation(
    Mesh Orientation Mode=::NewEnumerator2
    Rotation=&RandomRangeVector(Minimum=<0 0 0>, Maximum=<0 0 1>)
    Use Rotation Vector=true
   )
   &CameraOffset(Camera Offset Amount=2)
  ]
  particleUpdateScript [
   &ParticleState()
   &ScaleColor(
    Scale Alpha=&FloatFromCurve(
     FloatCurve=DataInterface() {propertyValues {Curve {keys [
         {arriveTangent=4.6626, interpMode=RCIM_Cubic, leaveTangent=4.6626, tangentMode=RCTM_User, time=0, value=0}
         {arriveTangent=0.037, interpMode=RCIM_Cubic, leaveTangent=0.037, tangentMode=RCTM_User, time=0.2, value=1}
         {arriveTangent=-1.7625, interpMode=RCIM_Cubic, leaveTangent=-1.7625, tangentMode=RCTM_User, time=0.6, value=0.8}
         <at 1 0 RCIM_Cubic>
        ]}}}
     Scale Curve=&Multiply_Float(
      A=&Multiply_Float(A=15, B=&FNTimeOfDayScalar(Night Scale=0.8, Morning Scale=1, Evening Scale=1))
      B=&Multiply_Float(A=&FN_FrontEndFloatModifier(Gameplay=1, Front End=0.5), B=1)
     )
    )
    Scale RGB=&MakeVectorFromLinearColorRGB(Color=&ColorFromCurve(ColorCurve=DataInterface() {propertyValues {
        AlphaCurve {keys []}
        BlueCurve {keys [<at 0.3585 0.0416>,<at 0.4586 1>,<at 0.4943 1>,<at 0.6047 0.0768 RCIM_Cubic>]}
        GreenCurve {keys [<at 0.3585 0.0098>,<at 0.4586 0.1914>,<at 0.4943 0.0819>,<at 0.6047 0.0319 RCIM_Cubic>]}
        RedCurve {keys [<at 0.3585 1>,<at 0.4586 0>,<at 0.4943 0>,<at 0.6047 1 RCIM_Cubic>]}
       }}))
    ScaleA=true
    ScaleRGB=true
   )
   &ScaleColor(
    Scale Alpha=&Lerp_Float(
     A=0.2
     B=&Multiply_Float(A=1, B=&Multiply_Float(A=1, B=$NPC.FortniteNPC.EventAudio_Amplitude))
     Alpha=$System.InterpolateOverTime.InterpolateOverTime_Float.Moving Average
    )
    ScaleA=true
    ScaleRGB=true
   ) {module=ScaleColor001}
  ]
  rendererClasses [&NiagaraMeshRendererProperties]
  renderers [{rendererClass=&NiagaraMeshRendererProperties, rendererIndex=0}]
 }
 Glints {
  emitterUpdateScript [
   &EmitterState(Life Cycle Mode=::NewEnumerator1, Loop Duration=1, UseLoopDelay=false)
   &SpawnRate(
    Spawn Probability=&Lerp_Float(
     A=0
     B=1
     Alpha=&FloatFromCurve(FloatCurve=DataInterface() {propertyValues {Curve {keys [<at 0 0>,<at 0.1 1>]}}}, CurveIndex=$System.Audio Reactive)
    )
    SpawnRate=3
    Use Spawn Probability=true
   )
  ]
  particleSpawnScript [
   &InitializeParticle(
    Color=<rgba 2 1.2079 0.9176 1.5>
    Lifetime Max=1
    Lifetime Min=0.5
    Lifetime Mode=::NewEnumerator1
    Sprite Size Mode=::NewEnumerator1
    Uniform Sprite Size Max=13
    Uniform Sprite Size Min=8
    UsePositionOffset=false
    Write Material Random=true
   )
   &ShapeLocation(
    MaskSpawnGroup=false
    Non Uniform Scale=<1.5 1 1>
    Offset=<0 -10 0.3>
    Offset Mode=::NewEnumerator0
    Override Seed=false
    Sphere Radius=6
    Sphere Surface Distribution=1
    UseOwnerScale=true
   )
   &CameraOffset(Camera Offset Amount=&RandomRangeFloat(Minimum=8, Maximum=12))
  ]
  particleUpdateScript [
   &ParticleState()
   &DynamicMaterialParameters(
    Index 0 Param 1=&FloatFromCurve(FloatCurve=DataInterface() {propertyValues {Curve {keys [{arriveTangent=-3.2272, interpMode=RCIM_Cubic, leaveTangent=-3.2272, tangentMode=RCTM_User, time=0, value=1}, <at 1 0 RCIM_Cubic>]}}})
    Index 0 Param 2=&FloatFromCurve(FloatCurve=DataInterface() {propertyValues {Curve {keys [{arriveTangent=-12.4095, interpMode=RCIM_Cubic, leaveTangent=-12.4095, tangentMode=RCTM_User, time=0.5, value=2}, <at 1 0 RCIM_Cubic>]}}})
    Module.Param0WriteEnabled=true
    Module.Param1WriteEnabled=true
    Module.Param2WriteEnabled=true
    Module.Param3WriteEnabled=true
   ) {module=DynamicMaterialParameters001}
   &ScaleSpriteSize(Scale Sprite Size Mode=::NewEnumerator0, Uniform Scale Factor=0.6)
   &ScaleColor(
    Scale Alpha=&FloatFromCurve(FloatCurve=DataInterface() {propertyValues {Curve {keys [
         {interpMode=RCIM_Cubic, tangentMode=RCTM_User, time=0, value=0}
         {interpMode=RCIM_Cubic, tangentMode=RCTM_User, time=0.1, value=1}
         {tangentMode=RCTM_User, time=1, value=1}
        ]}}})
    Scale RGB=&VectorFromFloat(Value=&Multiply_Float(
      A=&FloatFromCurve(
       FloatCurve=DataInterface() {propertyValues {Curve {keys [
           {arriveTangent=15.0225, interpMode=RCIM_Cubic, leaveTangent=15.0225, tangentMode=RCTM_User, time=0, value=0}
           {arriveTangentWeight=1.8705, interpMode=RCIM_Cubic, leaveTangentWeight=1.8705, tangentMode=RCTM_User, time=0.05, value=1}
           <at 1 0 RCIM_Cubic>
          ]}}}
       Scale Curve=3
      )
      B=&FNTimeOfDayScalar(Night Scale=0.5, Morning Scale=1, Evening Scale=1)
     ))
    ScaleA=true
    ScaleRGB=true
   )
  ]
  rendererClasses [&NiagaraSpriteRendererProperties]
  renderers [{rendererClass=&NiagaraSpriteRendererProperties}]
 }
 Glow {
  emitterUpdateScript [
   &EmitterState(
    Inactive Response=::NewEnumerator1
    Life Cycle Mode=::NewEnumerator1
    Loop Behavior=::NewEnumerator1
    Loop Duration Mode=::NewEnumerator1
    UseLoopDelay=false
   )
   &SpawnBurst_Instantaneous(Use Spawn Probability=false, UseLoopCountLimit=false)
  ]
  particleSpawnScript [
   &InitializeParticle(
    Color=<rgba 1 0.02 0 0.7>
    Mesh Scale=<1 0.88 1.2>
    Mesh Scale Mode=::NewEnumerator4
    Position Offset=<0 -15 0>
    Sprite Size Mode=::NewEnumerator3
    Uniform Sprite Size=70
    UsePositionOffset=true
    Write Material Random=true
   )
   &DynamicMaterialParameters(
    Index 0 Param 1=4
    Index 0 Param 2=2
    Module.Index0Float0Write=true
    Module.Index0Float1Write=true
    Module.Index0Float2Write=true
    Module.Index0Float3Write=true
   )
   &CameraOffset(Camera Offset Amount=-5)
   /SetVariables() {isSetParameters=true}
   &InitialMeshOrientation(Mesh Orientation Mode=::NewEnumerator2, Rotation=<0.03 0 0>, Use Rotation Vector=true)
  ]
  particleUpdateScript [
   &ParticleState(Kill Particles When Lifetime Has Elapsed=false)
   &ScaleMeshSize(Scale Factor=&Multiply_Vector(
     A=&VectorFromFloat(Value=1)
     B=&VectorFromFloat(Value=&MakeCustomFloatFromBool(
       Bool=&SetBoolByFloatComparison(A=$System.InterpolateOverTime.InterpolateOverTime_Float.Moving Average, B=0)
       True Float=1
       False Float=&MakeCustomFloatFromBool(Bool=$User.bIsFrontEndPreview, True Float=1, False Float=0)
      ))
    ))
   &SkeletalMeshLocation(Bone Sampling Mode=::NewEnumerator4, MaskSpawnGroup=false, Sampled Position Offset=<0 0.5 3>, Skeletal Mesh=$System.Skeletal Mesh)
   /SetVariables(Particles.ColorGlow:LinearColor=&Multiply_LinearColorByFloat(
     Linear Color=<rgba 1.5 1.5 1.5 1>
     Float=&MakeCustomFloatFromBool(
      Bool=$User.bIsFrontEndPreview
      True Float=&Lerp_Float(A=0.4, B=1, Alpha=$System.InterpolateOverTime.InterpolateOverTime_Float.Moving Average)
      False Float=&Lerp_Float(A=0, B=1, Alpha=$System.InterpolateOverTime.InterpolateOverTime_Float.Moving Average)
     )
    )) {isSetParameters=true}
   /SetVariables(Particles.ColorLight:LinearColor=&MakeLinearColorFromVectorAndFloat(
     "Vector (RGB)"=&Multiply_VectorByFloat(
      Vector=&MakeVectorFromLinearColorRGB(Color=<rgba 1 1 1 1>)
      Float=&Lerp_Float(
       A=1
       B=80
       Alpha=&Power(
        a=&Multiply_Float(A=$System.InterpolateOverTime.InterpolateOverTime_Float.Moving Average, B=$NPC.FortniteNPC.EventAudio_Amplitude)
        b=10
       )
      )
     )
     "Float (Alpha)"=&Lerp_Float(A=0, B=1, Alpha=$System.InterpolateOverTime.InterpolateOverTime_Float.Moving Average)
    )) {isSetParameters=true}
   &ScaleColor(
    Scale Alpha=&MakeCustomFloatFromBool(
     Bool=$User.bIsFrontEndPreview
     True Float=&Lerp_Float(A=0.2, B=1, Alpha=$System.InterpolateOverTime.InterpolateOverTime_Float.Moving Average)
     False Float=&Lerp_Float(A=0, B=1, Alpha=$System.InterpolateOverTime.InterpolateOverTime_Float.Moving Average)
    )
    ScaleA=true
    ScaleRGB=true
   )
   /SetVariables() {isSetParameters=true}
   /SetVariables(Particles.Size003:Vector3f=&VectorFromFloat(Value=1)) {isSetParameters=true}
   &FN_Light_Attributes(
    Color=<rgba 1 0.0872 0.0781 1>
    Create Color Attribute=true
    Create Exponent Attribute=true
    Create Position Attribute=true
    Create Radius Attribute=true
    Exponent=2
    Position:NiagaraPosition=$Particles.SkeletalMeshLocation.SampledPosition
    Radius=&Multiply_Float(A=50, B=&Lerp_Float(A=0, B=1, Alpha=$System.InterpolateOverTime.InterpolateOverTime_Float.Moving Average))
    Scale Color=20
   )
  ]
  rendererClasses [
   &NiagaraSpriteRendererProperties
   &NiagaraLightRendererProperties
   &NiagaraMeshRendererProperties
  ]
  renderers [
   {rendererClass=&NiagaraSpriteRendererProperties}
   {rendererClass=&NiagaraLightRendererProperties, rendererIndex=1}
   {rendererClass=&NiagaraMeshRendererProperties, rendererIndex=2}
   {rendererClass=&NiagaraMeshRendererProperties, rendererIndex=3}
  ]
 }
 MusicLine {
  emitterUpdateScript [
   &EmitterState(
    Inactive Response=::NewEnumerator1
    Life Cycle Mode=::NewEnumerator1
    Loop Behavior=::NewEnumerator1
    Loop Duration Mode=::NewEnumerator1
    UseLoopDelay=false
   )
   &SpawnBurst_Instantaneous(Spawn Count=2, Use Spawn Probability=false, UseLoopCountLimit=false)
  ]
  particleSpawnScript [
   &InitializeParticle(Color Mode=::NewEnumerator0, Lifetime=5, Mesh Scale Mode=::NewEnumerator4, UsePositionOffset=true, Write Material Random=true)
   &InitialMeshOrientation(Mesh Orientation Mode=::NewEnumerator2, Rotation=<0 0 0.5>, Use Rotation Vector=true)
   /SetVariables(Particles.MeshOri_In:Quat4f=<0 0 0.25 1>) {isSetParameters=true}
   /SetVariables(Particles.Position_Out:NiagaraPosition=<0 -12 -12>) {isSetParameters=true}
   /SetVariables(Particles.Position_In:NiagaraPosition=<0 -12 -10>) {isSetParameters=true}
  ]
  particleUpdateScript [
   &ParticleState(Kill Particles When Lifetime Has Elapsed=false, Let Infinitely Lived Particles Die When Emitter Deactivates=true)
   /SetVariables(Particles.Color_lightOut:LinearColor=&MakeLinearColorFromVectorAndFloat(
     "Vector (RGB)"=&Multiply_Vector(
      A=&MakeVectorFromLinearColorRGB(Color=<rgba 0.6667 0.0894 0.0833 1>)
      B=&VectorFromFloat(Value=&Lerp_Float(
        A=2
        B=30
        Alpha=&Power(
         a=&Multiply_Float(A=$System.InterpolateOverTime.InterpolateOverTime_Float.Moving Average, B=$NPC.FortniteNPC.EventAudio_Amplitude)
         b=10
        )
       ))
     )
     "Float (Alpha)"=&MakeCustomFloatFromBool(
      Bool=$User.bIsFrontEndPreview
      True Float=&Lerp_Float(A=0.4, B=1.3, Alpha=$System.InterpolateOverTime.InterpolateOverTime_Float.Moving Average)
      False Float=&Lerp_Float(A=0, B=1.3, Alpha=$System.InterpolateOverTime.InterpolateOverTime_Float.Moving Average)
     )
    )) {isSetParameters=true}
   /SetVariables(Particles.Color_BaseOut:LinearColor=&MakeLinearColorFromVectorAndFloat(
     "Vector (RGB)"=&Multiply_VectorByFloat(
      Vector=&MakeVectorFromLinearColorRGB(Color=<rgba 0.6823 0.0546 0.0671 1>)
      Float=&Lerp_Float(
       A=0.8
       B=3
       Alpha=&Power(
        a=&Multiply_Float(A=$System.InterpolateOverTime.InterpolateOverTime_Float.Moving Average, B=$NPC.FortniteNPC.EventAudio_Amplitude)
        b=10
       )
      )
     )
     "Float (Alpha)"=&MakeCustomFloatFromBool(
      Bool=$User.bIsFrontEndPreview
      True Float=&Lerp_Float(A=0.3, B=1.5, Alpha=$System.InterpolateOverTime.InterpolateOverTime_Float.Moving Average)
      False Float=&Lerp_Float(A=0, B=1.5, Alpha=$System.InterpolateOverTime.InterpolateOverTime_Float.Moving Average)
     )
    )) {isSetParameters=true}
   /SetVariables(Particles.Color_lightIn:LinearColor=&MakeLinearColorFromVectorAndFloat(
     "Vector (RGB)"=&Multiply_Vector(
      A=&MakeVectorFromLinearColorRGB(Color=<rgba 0.6667 0.0667 0.0764 1>)
      B=&VectorFromFloat(Value=&Lerp_Float(
        A=2.5
        B=&Multiply_Float(A=80, B=&FNTimeOfDayScalar(Night Scale=0.7, Morning Scale=0.8, Evening Scale=0.8))
        Alpha=&Power(
         a=&Multiply_Float(A=$System.InterpolateOverTime.InterpolateOverTime_Float.Moving Average, B=$NPC.FortniteNPC.EventAudio_Amplitude)
         b=10
        )
       ))
     )
     "Float (Alpha)"=&MakeCustomFloatFromBool(
      Bool=$User.bIsFrontEndPreview
      True Float=&Lerp_Float(A=0.4, B=1.3, Alpha=$System.InterpolateOverTime.InterpolateOverTime_Float.Moving Average)
      False Float=&Lerp_Float(A=0, B=1.3, Alpha=$System.InterpolateOverTime.InterpolateOverTime_Float.Moving Average)
     )
    )) {isSetParameters=true}
   /SetVariables(Particles.Color_BaseIn:LinearColor=&MakeLinearColorFromVectorAndFloat(
     "Vector (RGB)"=&Multiply_VectorByFloat(
      Vector=&MakeVectorFromLinearColorRGB(Color=<rgba 1 0.02 0.03 1>)
      Float=&Lerp_Float(
       A=0.8
       B=3
       Alpha=&Power(
        a=&Multiply_Float(A=$System.InterpolateOverTime.InterpolateOverTime_Float.Moving Average, B=$NPC.FortniteNPC.EventAudio_Amplitude)
        b=10
       )
      )
     )
     "Float (Alpha)"=&MakeCustomFloatFromBool(
      Bool=$User.bIsFrontEndPreview
      True Float=&Lerp_Float(A=0.2, B=1, Alpha=$System.InterpolateOverTime.InterpolateOverTime_Float.Moving Average)
      False Float=&Lerp_Float(A=0, B=1, Alpha=$System.InterpolateOverTime.InterpolateOverTime_Float.Moving Average)
     )
    )) {isSetParameters=true}
   &DynamicMaterialParameters(
    Index 0 Param 1=&Lerp_Float(
     A=&MakeCustomFloatFromBool(Bool=$User.bIsFrontEndPreview, True Float=0.06, False Float=0)
     B=0.15
     Alpha=$System.InterpolateOverTime.InterpolateOverTime_Float.Moving Average
    )
    Index 0 Param 2=&Lerp_Float(
     A=&MakeCustomFloatFromBool(Bool=$User.bIsFrontEndPreview, True Float=0.2, False Float=0)
     B=0.5
     Alpha=$System.InterpolateOverTime.InterpolateOverTime_Float.Moving Average
    )
    Module.Index0Float0Write=true
    Module.Index0Float1Write=true
    Module.Index0Float2Write=true
    Module.Index0Float3Write=true
   )
   /SetVariables(Particles.Size_Out:Vector3f=&Multiply_Vector(
     A=<0.88 0.88 1.8>
     B=&MakeVector(
      X=1
      Y=1
      Z=&Lerp_Float(
       A=&Lerp_Float(
        A=&MakeCustomFloatFromBool(Bool=$User.bIsFrontEndPreview, True Float=0.8, False Float=0.5)
        B=1.2
        Alpha=$System.InterpolateOverTime.InterpolateOverTime_Float.Moving Average
       )
       B=1.5
       Alpha=&Power(
        a=&Multiply_Float(A=$System.InterpolateOverTime.InterpolateOverTime_Float.Moving Average, B=$NPC.FortniteNPC.EventAudio_Amplitude)
        b=10
       )
      )
     )
    )) {isSetParameters=true}
   /SetVariables(Particles.Size_In:Vector3f=&Multiply_Vector(
     A=<0.8 0.8 2.3>
     B=&MakeVector(
      X=1
      Y=1
      Z=&Lerp_Float(
       A=&Lerp_Float(
        A=&MakeCustomFloatFromBool(Bool=$User.bIsFrontEndPreview, True Float=0.9, False Float=0.5)
        B=1.5
        Alpha=$System.InterpolateOverTime.InterpolateOverTime_Float.Moving Average
       )
       B=2
       Alpha=&Power(
        a=&Multiply_Float(A=$System.InterpolateOverTime.InterpolateOverTime_Float.Moving Average, B=$NPC.FortniteNPC.EventAudio_Amplitude)
        b=10
       )
      )
     )
    )) {isSetParameters=true}
  ]
  rendererClasses [&NiagaraMeshRendererProperties]
  renderers [
   {rendererClass=&NiagaraMeshRendererProperties, rendererIndex=0}
   {rendererClass=&NiagaraMeshRendererProperties, rendererIndex=1}
   {rendererClass=&NiagaraMeshRendererProperties, rendererIndex=2}
   {rendererClass=&NiagaraMeshRendererProperties, rendererIndex=3}
  ]
 }
 MusicLineGlow {
  emitterUpdateScript [
   &EmitterState(
    Inactive Response=::NewEnumerator1
    Life Cycle Mode=::NewEnumerator1
    Loop Behavior=::NewEnumerator1
    Loop Duration Mode=::NewEnumerator1
    UseLoopDelay=false
   )
   &SpawnBurst_Instantaneous(Spawn Count=2, Use Spawn Probability=false, UseLoopCountLimit=false)
  ]
  particleSpawnScript [
   &PartitionParticles()
   &InitializeParticle(
    Color Mode=::NewEnumerator0
    Lifetime=5
    Mesh Scale=&SelectVectorFromArray(
     Direct Array Index=$Particles.PartitionParticles.Partition
     Vector Selection Array=DataInterface() {propertyValues {FloatData [<0.85 0.85 0.8>, <-0.75 0.75 1>]}}
    )
    Mesh Scale Mode=::NewEnumerator4
    Position Offset=&SelectVectorFromArray(
     Direct Array Index=$Particles.PartitionParticles.Partition
     Vector Selection Array=DataInterface() {propertyValues {FloatData [<0 -12 -14>, <0 -12 -10>]}}
    )
    UsePositionOffset=true
    Write Material Random=true
   )
   &InitialMeshOrientation(Mesh Orientation Mode=::NewEnumerator2, Rotation=<0 0 0.5>, Use Rotation Vector=true)
   &CameraOffset(Camera Offset Amount=-2)
  ]
  particleUpdateScript [
   &ParticleState(Kill Particles When Lifetime Has Elapsed=false, Let Infinitely Lived Particles Die When Emitter Deactivates=true)
   /SetVariables(Particles.ColorGlow:LinearColor=&MakeLinearColorFromVectorAndFloat(
     "Vector (RGB)"=&MakeVectorFromLinearColorRGB(Color=&SelectLinearColorFromArray(
       Array Sampling Mode=::NewEnumerator1
       Direct Array Index=$Particles.PartitionParticles.Partition
       Color Selection Array=DataInterface() {propertyValues {ColorData [<rgba 1 0.099 0.1084 1>, <rgba 1 0.3542 0.361 1>]}}
       Random Seed=?
      ))
     "Float (Alpha)"=&Multiply_Float(
      A=&FloatFromCurve(
       FloatCurve=DataInterface() {propertyValues {Curve {keys [
           {interpMode=RCIM_Cubic, tangentMode=RCTM_User, time=0.1, value=0}
           {arriveTangent=7.2184, interpMode=RCIM_Cubic, tangentMode=RCTM_Break, time=0.4, value=1}
          ]}}}
       CurveIndex=$System.InterpolateOverTime.InterpolateOverTime_Float.Moving Average
       Scale Curve=0.6
      )
      B=1
     )
    )) {isSetParameters=true}
   /SetVariables(Particles.SizeGlow:Vector3f=&MakeVector(
     X=&SelectFloatFromArray(Direct Array Index=$Particles.PartitionParticles.Partition, Float Selection Array=DataInterface() {propertyValues {FloatData [0.8,-0.7]}})
     Y=&SelectFloatFromArray(Direct Array Index=$Particles.PartitionParticles.Partition, Float Selection Array=DataInterface() {propertyValues {FloatData [0.8,0.7]}})
     Z=&Lerp_Float(
      A=&SelectFloatFromArray(Direct Array Index=$Particles.PartitionParticles.Partition, Float Selection Array=DataInterface() {propertyValues {FloatData [1.3,1.5]}})
      B=3
      Alpha=&Power(
       a=&Multiply_Float(A=$System.InterpolateOverTime.InterpolateOverTime_Float.Moving Average, B=$NPC.FortniteNPC.EventAudio_Amplitude)
       b=10
      )
     )
    )) {isSetParameters=true}
   &DynamicMaterialParameters(
    Index 0 Param 1=&Lerp_Float(A=0, B=0.6, Alpha=$System.InterpolateOverTime.InterpolateOverTime_Float.Moving Average)
    Module.Index0Float0Write=true
    Module.Index0Float1Write=true
    Module.Index0Float2Write=true
    Module.Index0Float3Write=true
   )
  ]
  rendererClasses [&NiagaraMeshRendererProperties]
  renderers [{rendererClass=&NiagaraMeshRendererProperties, rendererIndex=0}]
 }
 Star {
  emitterUpdateScript [
   &EmitterState(Inactive Response=::NewEnumerator1, Life Cycle Mode=::NewEnumerator1, Loop Duration=7.6333, UseLoopDelay=false)
   &SpawnBurst_Instantaneous(Use Spawn Probability=false, UseLoopCountLimit=false) {enabled=false}
   &SpawnRate(
    Spawn Probability=&Lerp_Float(A=0, B=1, Alpha=&Lerp_Float(A=0, B=1, Alpha=$System.Audio Reactive))
    SpawnRate=2
    Use Spawn Probability=true
   )
  ]
  particleSpawnScript [
   &InitializeParticle(
    Color=<rgba 0.6667 0 0 2>
    Sprite Rotation Angle=&RandomRangeFloat(Minimum=-0.1, Maximum=0.1)
    Sprite Rotation Mode=::NewEnumerator1
    Sprite Size Mode=::NewEnumerator1
    Uniform Sprite Size Min=8
    UsePositionOffset=false
    Write Material Random=true
   )
   &ShapeLocation(
    Cylinder Height=1
    Cylinder Radius=15
    MaskSpawnGroup=false
    Offset=<-2 -12 0>
    Offset Mode=::NewEnumerator0
    Override Seed=false
    Shape Primitive=::NewEnumerator1
    Surface Only=true
    Use Endcaps In Surface Only Mode=false
    UseOwnerScale=true
   )
   &AddVelocity(Module.Use Falloff Distance=false, Origin Offset=<0 -10 -50>, Override Seed=false, Velocity Mode=::NewEnumerator1, Velocity Speed=10)
   &AddVelocity(Velocity=<0 0 10>) {enabled=false, module=AddVelocity001}
  ]
  particleUpdateScript [
   &ParticleState()
   &ScaleSpriteSize(
    Uniform Curve Scale=&Multiply_Float(
     A=&RandomRangeFloat(Minimum=0.8, Maximum=1)
     B=&Lerp_Float(
      A=&Lerp_Float(A=0, B=1, Alpha=$System.Audio Reactive)
      B=6
      Alpha=&Power(
       a=&Multiply_Float(A=$System.Audio Reactive, B=&ClampFloat(Float=$NPC.FortniteNPC.EventAudio_Amplitude, Min=0.5, Max=0.8))
       b=8
      )
     )
    )
    Uniform Curve Sprite Scale=DataInterface() {propertyValues {Curve {keys [
        {tangentMode=RCTM_User, time=0, value=0}
        <at 0.1 1>
        <at 0.3 0.65>
        <at 0.5 1>
        <at 0.7 0.7>
        <at 0.9 1>
        {tangentMode=RCTM_User, time=1, value=0}
       ]}}}
   ) {module=ScaleSpriteSize001}
   &Drag(Drag=5, Use Linear Drag=true, Use Rotational Drag=true)
   /SetVariables(Particles.ColorLight:LinearColor=&MakeLinearColorFromVectorAndFloat(
     "Vector (RGB)"=&Multiply_VectorByFloat(
      Vector=&MakeVectorFromLinearColorRGB(Color=&ColorFromCurve(ColorCurve=DataInterface() {propertyValues {
          AlphaCurve {keys []}
          BlueCurve {keys [<at 0.1054 0.3185 RCIM_Cubic>,<at 0.6771 0.1367 RCIM_Cubic>]}
          GreenCurve {keys [<at 0.1054 0.5732 RCIM_Cubic>,<at 0.6771 0.1028 RCIM_Cubic>]}
          RedCurve {keys [<at 0.1054 1 RCIM_Cubic>,<at 0.6771 1 RCIM_Cubic>]}
         }}))
      Float=&Lerp_Float(
       A=1
       B=25
       Alpha=&Power(a=&Multiply_Float(A=$System.Audio Reactive, B=$NPC.FortniteNPC.EventAudio_Amplitude), b=10)
      )
     )
     "Float (Alpha)"=&FloatFromCurve(FloatCurve=DataInterface() {propertyValues {Curve {keys [<at 0 1>,<at 0.8 1>,<at 1 0>]}}})
    )) {isSetParameters=true}
   /SetVariables(Particles.ColorGlow:LinearColor=&MakeLinearColorFromVectorAndFloat(
     "Vector (RGB)"=&MakeVectorFromLinearColorRGB(Color=<rgba 1 0.1198 0.1198 1>)
     "Float (Alpha)"=&FloatFromCurve(
      FloatCurve=DataInterface() {propertyValues {Curve {keys [<at 0 0>,<at 0.2 1>,<at 0.6 1>,<at 1 0>]}}}
      Scale Curve=&Lerp_Float(A=0, B=0.8, Alpha=$System.InterpolateOverTime.InterpolateOverTime_Float.Moving Average)
     )
    )) {isSetParameters=true}
   /SetVariables(Particles.SpriteSizeGlow:Vector2f=<15 15>) {isSetParameters=true}
   &ScaleColor(
    Scale Alpha=&FloatFromCurve(FloatCurve=DataInterface() {propertyValues {Curve {keys [<at 0 1>,<at 0.8 1>,<at 1 0>]}}})
    ScaleA=true
    ScaleRGB=true
   )
   &SolveForcesAndVelocity(Clamp Velocity=false, Limit Acceleration=false)
   &DynamicMaterialParameters(
    Index 0 Param 3=&Multiply_Float(
     A=&FloatFromCurve(FloatCurve=DataInterface() {propertyValues {Curve {keys [<at 0 1>,<at 1 1>]}}}, Scale Curve=6)
     B=&RandomRangeFloat(Minimum=-1, Maximum=1.5)
    )
    Module.Index0Float0Write=false
    Module.Index0Float1Write=false
    Module.Index0Float2Write=true
    Module.Index0Float3Write=true
   )
  ]
  rendererClasses [&NiagaraSpriteRendererProperties]
  renderers [
   {rendererClass=&NiagaraSpriteRendererProperties}
   {rendererClass=&NiagaraSpriteRendererProperties, rendererIndex=1}
   {rendererClass=&NiagaraSpriteRendererProperties, rendererIndex=2}
  ]
 }
 spikes {
  emitterUpdateScript [
   &EmitterState(Inactive Response=::NewEnumerator1, Life Cycle Mode=::NewEnumerator1, Loop Delay=0.03, Loop Duration=1, UseLoopDelay=false)
   &SpawnRate(
    Spawn Probability=&Lerp_Float(
     A=0
     B=1
     Alpha=&Lerp_Float(
      A=0
      B=1
      Alpha=&FloatFromCurve(FloatCurve=DataInterface() {propertyValues {Curve {keys [<at 0 0>,<at 0.1 1>]}}}, CurveIndex=$System.Audio Reactive)
     )
    )
    SpawnRate=10
    Use Spawn Probability=true
   )
  ]
  particleSpawnScript [
   &InitializeParticle(
    Lifetime Max=0.5
    Lifetime Min=0.3
    Lifetime Mode=::NewEnumerator1
    Position Offset=<10 0 0>
    Sprite Size Max=<4 25>
    Sprite Size Min=<3 15>
    Sprite Size Mode=::NewEnumerator2
    UsePositionOffset=false
    Write Material Random=true
   )
   &ShapeLocation(
    Cylinder Height=5
    Cylinder Radius=10
    MaskSpawnGroup=false
    Offset=<0 -10 -8>
    Offset Mode=::NewEnumerator0
    Override Seed=false
    Shape Primitive=::NewEnumerator1
    Surface Only=true
    Use Endcaps In Surface Only Mode=false
    UseOwnerScale=true
   )
   &SubUVAnimation(End Frame=3, SubUV Animation Mode:ENiagara_SubUVLookupMode=::NewEnumerator2)
   &AddVelocity(
    Module.Use Falloff Distance=false
    Origin Offset=<0 -10 -40>
    Override Seed=false
    Velocity Mode=::NewEnumerator1
    Velocity Speed=&RandomRangeFloat(Minimum=50, Maximum=80)
   )
   &CameraOffset(Camera Offset Amount=2)
  ]
  particleUpdateScript [
   &ParticleState()
   &Drag(Drag=6, Use Linear Drag=true, Use Rotational Drag=false)
   &SolveForcesAndVelocity(Clamp Velocity=false, Limit Acceleration=false)
   &ScaleSpriteSize(
    Non-Uniform Curve Scale=&Multiply_Vector2D_ByFloat(Float=1, Vector2D=<1 1>)
    Non-Uniform Curve Sprite Scale=DataInterface() {propertyValues {
      XCurve {keys [{tangentMode=RCTM_User, time=0, value=1}, {tangentMode=RCTM_User, time=1, value=0.8}]}
      YCurve {keys [{tangentMode=RCTM_User, time=0, value=0.4}, {tangentMode=RCTM_User, time=1, value=1.2}]}
     }}
    Scale Sprite Size Mode=::NewEnumerator3
   )
   &ScaleColor(
    Scale Alpha=&FloatFromCurve(
     FloatCurve=DataInterface() {propertyValues {Curve {keys [<at 0 0 RCIM_Cubic>,<at 0.4 1 RCIM_Cubic>,<at 1 0 RCIM_Cubic>]}}}
     Scale Curve=&Lerp_Float(A=0, B=1.5, Alpha=$System.InterpolateOverTime.InterpolateOverTime_Float.Moving Average)
    )
    Scale RGB=&Multiply_VectorByFloat(
     Vector=&MakeVectorFromLinearColorRGB(Color=&ColorFromCurve(ColorCurve=DataInterface() {propertyValues {
         AlphaCurve {keys []}
         BlueCurve {keys [<at 0.1318 0.2576>, {arriveTangent=-0.3488, interpMode=RCIM_Cubic, time=0.4373, value=0.05}, <at 0.8381 0.0014>]}
         GreenCurve {keys [<at 0.1318 0.3149>, {arriveTangent=-0.4244, interpMode=RCIM_Cubic, time=0.4373, value=0.05}, <at 0.8381 0.0032>]}
         RedCurve {keys [<at 0.1318 1>, {arriveTangent=-1.1771, interpMode=RCIM_Cubic, time=0.4373, value=1}, <at 0.8381 0.1354>]}
        }}))
     Float=1.5
    )
    ScaleA=true
    ScaleRGB=true
   )
  ]
  rendererClasses [&NiagaraSpriteRendererProperties]
  renderers [{rendererClass=&NiagaraSpriteRendererProperties}]
 }
}
summary {
 systemName=NS_Backpack_ChimeCurlCorn_Idle
 emitters [
  {emitterName=MusicLineGlow, rendererClasses [&NiagaraMeshRendererProperties]}
  {
   emitterName=Glow
   rendererClasses [
    &NiagaraSpriteRendererProperties
    &NiagaraLightRendererProperties
    &NiagaraMeshRendererProperties
   ]
  }
  {emitterName=BoxGlow, rendererClasses [&NiagaraMeshRendererProperties]}
  {emitterName=Glints, rendererClasses [&NiagaraSpriteRendererProperties]}
  {emitterName=spikes, rendererClasses [&NiagaraSpriteRendererProperties]}
  {emitterName=Star, rendererClasses [&NiagaraSpriteRendererProperties]}
  {emitterName=MusicLine, rendererClasses [&NiagaraMeshRendererProperties]}
 ]
 userVariables [
  User.bIsFrontEnd {type=NiagaraBool}
  User.bIsFrontEndPreview {type=NiagaraBool}
 ]
}
