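# Parent material for the ChimeCurlCorn backpack cosmetic, built on the 2023 character PreFX and PostFX functions with switchable, mask-blended fur.
# An audio-reactivity switch overrides base color and emissive, using amplitude from the frontend-loop audio function.
# A powered amplitude drives pink emissive through the Tech Art Mask's blue channel.
# Named reroutes also carry a star-outline mask, built from local height and Metallic, and a saturation-boosted base color blend.
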Material /BRCosmetics/Accessories/FORT_Backpacks/Backpack_F_MED_ChimeCurlCorn/Materials/M_ChimeCurlCorn_Backpack_Parent
graph {
 properties {
  bUseMaterialAttributes=true
  bUsedWithInstancedStaticMeshes=true
  bUsedWithMeshDeformer=true
  bUsedWithMorphTargets=true
  bUsedWithSkeletalMesh=true
  pixelDepthOffsetMode=PDOM_Legacy
  shadingModel=MSM_FromMaterialExpression
  usedShadingModels=Default Lit | Subsurface | Clear Coat | Subsurface Profile | Cloth
 }
 nodes [
  Constant() {r=1}
  n1 = MaterialFunctionCall() {materialFunction=&MF_CharacterParent_2023_PreFX}
  n2 = MaterialFunctionCall!(In=@n1.Result, Mask Input=@n1.Result) {materialFunction=/Game/Global/Materials/Character_Parent_2023/Fur_Parent_2023/MF_Fur_Deprecated}
  n3 = GetMaterialAttributes(None=BlendMaterialAttributes(
    A=@n1.Result
    B=StaticSwitchParameter(True=@n2.Result, False=@n1.Result) {defaultValue=true, group=Skin FX, parameterName=UseFNHair2_0}
    Alpha=@n2.HairMask
   )) {attributeGetTypes [#1, #2, #3]}
  n4 = NamedRerouteUsage()
  n5 = NamedRerouteUsage()
  MaterialFunctionCall(
   Material=StaticSwitchParameter(
    True=SetMaterialAttributes!(MaterialAttributes=@n3.MaterialAttributes, Emissive Color=@n4, Base Color=@n5) {attributeSetTypes [#1, #3]}
    False=@n3.MaterialAttributes
   ) {defaultValue=true, group=AudioReactivity, parameterName=UseAudioReactivity}
   M=@n1.M
   Pre Skinned Local Position=@n1.M
   Distance Blend=@n1.M
   Fresnel=@n1.M
  ) {materialFunction=&MF_CharacterParent_2023_PostFX}
  n6 = Add(A=ComponentMask(None=@n1.Pre skinned Local Position) {b=true}) {constB=3}
  n7 = Constant() {r=0.6}
  NamedRerouteDeclaration(None=Multiply(A=Saturate(None=@n6), B=SphereMask(A=@n3.Metallic, B=@n7) {attenuationRadius=0.04})) {name=StarOutlineMask, nodeColor=<rgba 0 1 0.0625 1>}
  n8 = MaterialFunctionCall!(InAmplitude, InSpectrum, CustomFakeSpectrum, CustomFakeAmplitude) {materialFunction=/BRCosmetics/Global/MaterialFunctions/AudioReactivity/MF_AudioReactivity_FrontendLoop}
  n9 = Power(Base=@n8.Amplitude) {constExponent=8}
  NamedRerouteDeclaration(None=@n9) {name=AudioAmplitude_Modified, nodeColor=<rgba 1 0 0.2578 1>}
  NamedRerouteDeclaration(None=@n3.Base Color) {name=OriginalBaseColor, nodeColor=<rgba 0 1 0.8594 1>}
  n10 = NamedRerouteUsage()
  n11 = NamedRerouteUsage()
  n12 = Add(A=@n10, B=Multiply(A=Multiply(A=@n11) {constB=0.35}, B=NamedRerouteUsage()))
  n13 = Constant() {r=-1}
  n14 = NamedRerouteUsage()
  n15 = Ceil(None=NamedRerouteUsage())
  NamedRerouteDeclaration(None=LinearInterpolate(A=@n10, B=Desaturation(None=@n12, Fraction=@n13), Alpha=Multiply(A=@n14, B=@n15))) {name=OutputBaseColor, nodeColor=<rgba 1 0 0.4219 1>}
  NamedRerouteDeclaration(None=@n3.Emissive Color) {name=OriginalEmissive, nodeColor=<rgba 0.7344 1 0 1>}
  n16 = ComponentMask(None=@n1.Tech Art Mask) {b=true}
  n17 = Multiply(A=Multiply(A=Constant3Vector() {constant=<rgba 1 0.4913 0.8594 1>}, B=Add(B=@n9) {constA=0.85}), B=Ceil(None=@n8.Amplitude))
  NamedRerouteDeclaration(None=Add(A=NamedRerouteUsage(), B=Multiply(A=@n16, B=@n17))) {name=OutputEmissive, nodeColor=<rgba 0.9453 0 1 1>}
 ]
}
